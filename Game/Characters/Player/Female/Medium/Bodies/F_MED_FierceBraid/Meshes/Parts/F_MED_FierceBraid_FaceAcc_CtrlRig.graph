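# Corrective Control Rig for the FierceBraid female body's face accessory, offsetting braid and earring bones based on upper-arm pose.
# Spherical pose readers on each upper arm feed AlphaInterp values used as Modify Transforms weights.
# The wired right branch shifts and rotates the right braid base; wired earring offsets are zero.
# Left braid-base and lower right-braid modifiers have no incoming execution link, so they are inactive.
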ControlRigBlueprint /Game/Characters/Player/Female/Medium/Bodies/F_MED_FierceBraid/Meshes/Parts/F_MED_FierceBraid_FaceAcc_CtrlRig
graphs [
 RigVMFunctionLibrary {links [], nodes []}
 RigVMModel {nodes [
   n1 = &RigUnit_SphericalPoseReader(
    ActiveRegionScaleFactors=(PositiveWidth=0.5,NegativeWidth=0.5,PositiveHeight=0.5,NegativeHeight=0.5)
    ActiveRegionSize=0.5
    Debug=(bDrawDebug=false,bDrawLocalAxes=false,DebugScale=25,DebugSegments=20,DebugThickness=0.25)
    DriverAxis=(X=1,Y=0,Z=0)
    DriverItem=(Type=Bone,Name="upperarm_r")
    FalloffRegionScaleFactors=(PositiveWidth=1,NegativeWidth=1,PositiveHeight=1,NegativeHeight=1)
    FalloffSize=0.5
    FlipWidthScaling=true
    OptionalParentItem=(Type=Bone)
    RotationOffset=(X=90,Y=0,Z=60)
    ExecutePin=@SequenceAggregate_3.B
   )
   n2 = &RigVMFunction_AlphaInterp(InterpSpeedDecreasing=10, InterpSpeedIncreasing=100, Scale=1, bClampResult=False, bInterpResult=False, Value=@n1.OutputParam)
   &RigUnit_ModifyTransforms(
    ExecutePin=()
    ItemToModify=((Item=(Type=Bone,Name="dyn_head_tail_2_r"),Transform=(Rotation=(X=0,Y=0,Z=0,W=1),Translation=(X=0,Y=0,Z=0),Scale3D=(X=1,Y=1,Z=1))))
    WeightMinimum=0
    ItemToModify.0.Transform.Rotation=&RigVMFunction_MathQuaternionFromEuler(Euler=(X=0,Y=0,Z=0), RotationOrder=ZXY).Result
    Weight=@n2.Result
   )
   > &RigUnit_ModifyTransforms(
    ItemToModify=((Item=(Type=Bone,Name="dyn_head_tail_3_r"),Transform=(Rotation=(X=0,Y=0,Z=0,W=1),Translation=(X=0,Y=0,Z=0),Scale3D=(X=1,Y=1,Z=1))))
    WeightMinimum=0
    ItemToModify.0.Transform.Rotation=&RigVMFunction_MathQuaternionFromEuler(Euler=(X=0,Y=0,Z=0), RotationOrder=ZXY).Result
    Weight=@n2.Result
   )
   RigUnit_BeginExecution = &RigUnit_BeginExecution()
   SequenceAggregate_3 = RigVMAggregateNode(ExecuteContext=@RigUnit_BeginExecution.ExecutePin)
   n3 = &RigUnit_SphericalPoseReader(
    ActiveRegionScaleFactors=(PositiveWidth=2.5,NegativeWidth=2.5,PositiveHeight=2.5,NegativeHeight=0.25)
    ActiveRegionSize=0.1
    Debug=(bDrawDebug=false,bDrawLocalAxes=false,DebugScale=25,DebugSegments=20,DebugThickness=0.25)
    DriverAxis=(X=-1,Y=0,Z=0)
    DriverItem=(Type=Bone,Name="upperarm_l")
    FalloffRegionScaleFactors=(PositiveWidth=1,NegativeWidth=1,PositiveHeight=1,NegativeHeight=1)
    FalloffSize=0.5
    FlipWidthScaling=true
    OptionalParentItem=(Type=Bone)
    RotationOffset=(X=90,Y=0,Z=0)
    ExecutePin=@SequenceAggregate_3.A
   )
   n4 = &RigVMFunction_AlphaInterp(InterpSpeedDecreasing=100, InterpSpeedIncreasing=100, Scale=1, bClampResult=False, bInterpResult=False, Value=@n3.OutputParam)
   > &RigUnit_ModifyTransforms(
    ItemToModify=((Item=(Type=Bone,Name="dyn_earrings_base_l"),Transform=(Rotation=(X=0,Y=0,Z=0,W=1),Translation=(X=0,Y=0,Z=0),Scale3D=(X=1,Y=1,Z=1))))
    WeightMinimum=0
    Weight=@n4.Result
    ItemToModify.0.Transform.Rotation=&RigVMFunction_MathQuaternionFromEuler(Euler=(X=0,Y=0,Z=0), RotationOrder=ZXY).Result
   )
   &RigUnit_ModifyTransforms(
    ExecutePin=()
    ItemToModify=((Item=(Type=Bone,Name="dyn_head_tail_base_l"),Transform=(Rotation=(X=0,Y=0,Z=0,W=1),Translation=(X=0,Y=5,Z=0),Scale3D=(X=1,Y=1,Z=1))))
    WeightMinimum=0
    Weight=@n4.Result
    ItemToModify.0.Transform.Rotation=&RigVMFunction_MathQuaternionFromEuler(Euler=(X=0,Y=0,Z=0), RotationOrder=ZXY).Result
   )
   &RigUnit_ModifyTransforms(
    ItemToModify=((Item=(Type=Bone,Name="dyn_earrings_base_r"),Transform=(Rotation=(X=0,Y=0,Z=0,W=1),Translation=(X=0,Y=0,Z=0),Scale3D=(X=1,Y=1,Z=1))))
    Weight=1
    WeightMinimum=0
    ItemToModify.0.Transform.Rotation=&RigVMFunction_MathQuaternionFromEuler(Euler=(X=0,Y=0,Z=0), RotationOrder=ZXY).Result
    ExecutePin=@n6.ExecutePin
   )
   n5 = &RigVMFunction_MathQuaternionFromEuler(Euler=(X=0,Y=20,Z=0), RotationOrder=ZXY)
   n6 = &RigUnit_ModifyTransforms(
    ItemToModify=((Item=(Type=Bone,Name="dyn_head_tail_base_r"),Transform=(Rotation=(X=0,Y=0,Z=0,W=1),Translation=(X=0,Y=-2,Z=-1),Scale3D=(X=1,Y=1,Z=1))))
    WeightMinimum=0
    Weight=@n2.Result
    ItemToModify.0.Transform.Rotation=@n5.Result
    ExecutePin=@n1.ExecutePin
   )
  ]}
 SequenceAggregate_3_ContainedGraph {nodes [
   n1 = RigVMFunctionEntryNode()
   SequenceAggregate = &RigVMFunction_Sequence(ExecuteContext=@n1.ExecuteContext)
   SequenceAggregate_1 = &RigVMFunction_Sequence(ExecuteContext=@SequenceAggregate.B)
   SequenceAggregate_2 = &RigVMFunction_Sequence(ExecuteContext=@SequenceAggregate_1.B)
   RigVMFunctionReturnNode(A=@SequenceAggregate.A, B=@SequenceAggregate_1.A, C=@SequenceAggregate_2.A, D=@SequenceAggregate_2.B)
  ]}
]
hierarchy [
 root {loc=<0 0 0>, rot=<0 0 0 1>}
 pelvis {parent=root, loc=<0 -0.5301 98.0571>, rot=<0.0127 0.707 -0.0127 -0.707>}
 spine_01 {parent=pelvis, loc=<6.1208 -0.9328 0>, rot=<0 0 0.034 -0.9994>}
 spine_02 {parent=spine_01, loc=<6.5996 -0.2095 0>, rot=<0 0 0.0238 -0.9997>}
 spine_03 {parent=spine_02, loc=<6.4914 -0.1237 0>, rot=<0 0 0.0149 -0.9999>}
 spine_04 {parent=spine_03, loc=<6.9598 0.0794 0>, rot=<0 0 -0.0132 -0.9999>}
 spine_05 {parent=spine_04, loc=<6.9987 0 0>, rot=<0 0 -0.0015 -1>}
 neck_01 {parent=spine_05, loc=<11.3138 0.1529 0>, rot=<0 0 -0.2003 -0.9797>}
 neck_02 {parent=neck_01, loc=<4.5627 0 0>, rot=<0 0 0.0504 -0.9987>}
 head {parent=neck_02, loc=<5.3914 0 0>, rot=<0 0 0.115 -0.9934>}
 dyn_bangs {parent=head, loc=<9.3383 10.3822 0.639>, rot=<0.0285 -0.9978 -0.0523 -0.0291>}
 dyn_curl {parent=head, loc=<7.9161 6.7295 -6.3682>, rot=<0.0532 -0.9921 -0.0976 -0.0585>}
 dyn_head_tail_base_l {parent=head, loc=<0.5858 -3.7038 -5.9316>, rot=<-0.2146 -0.9742 -0.0157 -0.0685>}
 dyn_head_tail_1_l {parent=dyn_head_tail_base_l, loc=<0 0 0>, rot=<0 0 0 1>}
 dyn_head_tail_2_l {parent=dyn_head_tail_1_l, loc=<6.4698 0.5074 0.1262>, rot=<-0.0785 -0.141 -0.0649 -0.9848>}
 dyn_head_tail_3_l {parent=dyn_head_tail_2_l, loc=<9.0035 -0.6969 0.3535>, rot=<0.0346 -0.0109 -0.0538 -0.9979>}
 dyn_head_tail_4_l {parent=dyn_head_tail_3_l, loc=<9.287 -0.0918 0.1245>, rot=<0.019 0.0274 0.0363 -0.9988>}
 dyn_head_tail_base_r {parent=head, loc=<0.5854 -3.7038 5.9316>, rot=<0.9692 -0.2361 -0.0681 0.0172>}
 dyn_head_tail_1_r {parent=dyn_head_tail_base_r, loc=<0 0 0>, rot=<0 0 0 1>}
 dyn_head_tail_2_r {parent=dyn_head_tail_1_r, loc=<-7.3533 -0.3036 -0.1188>, rot=<-0.0833 -0.1382 -0.099 -0.9819>}
 dyn_head_tail_3_r {parent=dyn_head_tail_2_r, loc=<-9.1962 0.4533 -0.2152>, rot=<0.0896 0.0395 -0.1586 -0.9825>}
 dyn_head_tail_4_r {parent=dyn_head_tail_3_r, loc=<-8.8473 0.1063 0.0006>, rot=<0.019 0.0274 0.0363 -0.9988>}
 dyn_earrings_base_l {parent=head, loc=<0.5752 0.0532 -5.7995>, rot=<0.0041 -0.9841 0.1773 -0.0007>}
 dyn_earrings_l {parent=dyn_earrings_base_l, loc=<0 0 0>, rot=<0 0 0 1>}
 dyn_earrings_base_r {parent=head, loc=<0.5754 0.0532 5.7995>, rot=<0.9841 0.0041 -0.0007 -0.1773>}
 dyn_earrings_r {parent=dyn_earrings_base_r, loc=<0 0 0>, rot=<0 0 0 1>}
 clavicle_l {parent=spine_05, loc=<5.4831 -0.1053 -3.2946>, rot=<0.0465 -0.7289 -0.009 -0.6829>}
 upperarm_l {parent=clavicle_l, loc=<9.8658 -1.2136 -0.4105>, rot=<0.0164 -0.3245 -0.0269 -0.9454>}
 lowerarm_l {parent=upperarm_l, loc=<26.465 -0.8345 0.259>, rot=<0.0036 0.0161 -0.2298 -0.9731>}
 clavicle_r {parent=spine_05, loc=<5.4831 -0.1053 3.2946>, rot=<0.7289 0.0465 -0.6829 0.009>}
 upperarm_r {parent=clavicle_r, loc=<-9.8658 1.2136 0.4105>, rot=<0.0164 -0.3245 -0.0269 -0.9454>}
 lowerarm_r {parent=upperarm_r, loc=<-26.465 0.8345 -0.259>, rot=<0.0036 0.0161 -0.2298 -0.9731>}
 hand_r {parent=lowerarm_r, loc=<-22.7355 0.3203 0.6106>, rot=<-0.6784 -0.0223 0.0735 -0.7307>}
 dyn_dragon_head {parent=spine_05, loc=<9.8406 4.1931 -13.0604>, rot=<-0.2743 -0.9473 0.0638 0.1529>}
 dyn_dragon_spine_1 {parent=spine_05, loc=<12.7987 -2.4621 -11.3692>, rot=<-0.1627 0.7191 0.0958 -0.6688>}
 dyn_dragon_spine_2 {parent=spine_05, loc=<12.2033 -6.5584 -0.838>, rot=<0.1235 0.6708 -0.1897 -0.7062>}
 dyn_dragon_spine_3 {parent=spine_05, loc=<12.5688 -0.9167 10.8801>, rot=<-0.4289 -0.9033 0.004 0.007>}
 dyn_dragon_spine_4 {parent=spine_05, loc=<4.7935 8.618 10.9928>, rot=<-0.2086 -0.929 -0.0907 -0.2919>}
 dyn_dragon_spine_5 {parent=spine_05, loc=<-3.6288 13.615 5.2698>, rot=<0.0791 -0.8764 0.0047 -0.4751>}
 dyn_dragon_spine_6 {parent=spine_05, loc=<-11.9192 11.4022 -7.5968>, rot=<0.4642 -0.7533 0.2118 -0.4149>}
]
