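# Physics asset for the face accessory mesh of the medium female FierceBraid player body, providing capsule bodies for its bangs, curl, earrings, head tails and dragon.
# Bangs, curl and earrings constrain to the head with limited swing and twist.
# Left and right head tails chain from angularly free bases on spine_05 through four segments with asymmetric limits.
# Upper-body bones also carry capsules, while dragon head and spine bodies have no constraints.
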PhysicsAsset /Game/Characters/Player/Female/Medium/Bodies/F_MED_FierceBraid/Meshes/Parts/F_MED_FierceBraid_FaceAcc_Physics
bodies {
 clavicle_l [{shapeType=Capsule, length=5.5, radius=7.5, shapeName=clavicle_l_capsule, center=<5.3349 0 0>, rotation=<rot -90 0 -0.0003>}]
 clavicle_r [{shapeType=Capsule, length=5.5, radius=7.5, shapeName=clavicle_r_capsule, center=<-5.2169 0 0>, rotation=<rot -90 174.9488 -174.9491>}]
 dyn_bangs [{shapeType=Capsule, length=6.5, radius=2, shapeName=dyn_bangs_capsule, center=<4.3448 -0.7972 0>}]
 dyn_curl [{shapeType=Capsule, length=6.5, radius=1, shapeName=dyn_curl_capsule, center=<5.1593 0 0>, rotation=<rot -90 0.005 -0.0048>}]
 dyn_dragon_head [
  {
   shapeType=Capsule
   length=12
   radius=7
   shapeName=dyn_dragon_head_capsule
   center=<1.8567 -0.221 3.1768>
   rotation=<rot -90 -160.529 160.5301>
  }
  {
   shapeType=Capsule
   length=8.5
   radius=5.5
   shapeName=dyn_dragon_head_capsule
   center=<-1.6744 6.2945 0.2362>
   rotation=<rot -85 -0.0003 5.0005>
  }
 ]
 dyn_dragon_spine_1 [{
   shapeType=Capsule
   length=7
   radius=6
   shapeName=dyn_dragon_spine_1_capsule
   center=<3.2429 -1.4202 0.2483>
   rotation=<rot 0.0011 -89.9999 89.9996>
  }]
 dyn_dragon_spine_2 [{
   shapeType=Capsule
   length=8.5
   radius=4.5
   shapeName=dyn_dragon_spine_2_capsule
   center=<5.9833 -2.9902 0>
   rotation=<rot 0.0006 -90 90.0003>
  }]
 dyn_dragon_spine_3 [{
   shapeType=Capsule
   length=6.5
   radius=3
   shapeName=dyn_dragon_spine_3_capsule
   center=<6.0725 -1.1929 -1.2553>
   rotation=<rot 0 -109.9986 90>
  }]
 dyn_dragon_spine_4 [{
   shapeType=Capsule
   length=5
   radius=4.5
   shapeName=dyn_dragon_spine_4_capsule
   center=<4.2346 0.4237 -1.8741>
   rotation=<rot -0.0004 -89.9994 94.9998>
  }]
 dyn_dragon_spine_5 [{
   shapeType=Capsule
   length=10
   radius=4
   shapeName=dyn_dragon_spine_5_capsule
   center=<7.0031 0.1726 -0.5245>
   rotation=<rot -1.2949 -85.168 79.9455>
  }]
 dyn_dragon_spine_6 [{
   shapeType=Capsule
   length=5
   radius=5
   shapeName=dyn_dragon_spine_6_capsule
   center=<5.7613 3.4292 0>
   rotation=<rot -0.0006 -59.9996 90.0003>
  }]
 dyn_earrings_l [{shapeType=Capsule, length=2, radius=1.5, shapeName=dyn_earrings_l_capsule, center=<2.6028 0 -0.6499>, rotation=<rot 90 180 89.9993>}]
 dyn_earrings_r [{shapeType=Capsule, length=2, radius=1.5, shapeName=dyn_earrings_l_capsule, center=<-2.6028 0 0.6499>, rotation=<rot -90 175.2364 -85.2357>}]
 dyn_head_tail_1_l [{
   shapeType=Capsule
   length=4
   radius=1.5
   shapeName=dyn_head_tail_1_l_capsule
   center=<3.7687 -0.3484 0.3022>
   rotation=<rot 85 0.0005 -0.0001>
  }]
 dyn_head_tail_1_r [{
   shapeType=Capsule
   length=3.5
   radius=1.5
   shapeName=dyn_head_tail_1_r_capsule
   center=<-3.7305 0.2115 -0.1336>
   rotation=<rot 90 -0.0009 -0.0017>
  }]
 dyn_head_tail_2_l [{
   shapeType=Capsule
   length=5.5
   radius=1.5
   shapeName=dyn_head_tail_2_l_capsule
   center=<4.3027 -0.2861 0>
   rotation=<rot 90 179.7629 179.7623>
  }]
 dyn_head_tail_2_r [{
   shapeType=Capsule
   length=6.5
   radius=1.5
   shapeName=dyn_head_tail_2_r_capsule
   center=<-4.3631 0.5873 -0.2523>
   rotation=<rot 90 6.7212 6.7206>
  }]
 dyn_head_tail_3_l [{shapeType=Capsule, length=5, radius=1.5, shapeName=dyn_head_tail_3_l_capsule, center=<4.3027 0 0>, rotation=<rot 90 166.7373 166.7369>}]
 dyn_head_tail_3_r [{shapeType=Capsule, length=5.5, radius=2, shapeName=dyn_head_tail_3_r_capsule, center=<-4.0933 0.0153 0>, rotation=<rot 90 0.1582 0.1578>}]
 dyn_head_tail_4_l [{shapeType=Capsule, length=3.75, radius=2.75, shapeName=dyn_head_tail_4_l_capsule, center=<4.3027 0 0>, rotation=<rot 90 -160.529 -160.5294>}]
 dyn_head_tail_4_r [{shapeType=Capsule, length=3.5, radius=3, shapeName=dyn_head_tail_4_r_capsule, center=<-4.5119 0.4074 0>, rotation=<rot 90 -90 -90.0005>}]
 dyn_head_tail_base_l [{shapeType=Capsule, length=0, radius=1.5, shapeName=dyn_head_tail_base_l_capsule}]
 dyn_head_tail_base_r [{shapeType=Capsule, length=2, radius=1.5, shapeName=dyn_head_tail_base_r_capsule}]
 hand_r [{shapeType=Capsule, length=4.75, radius=5, shapeName=hand_r_capsule, center=<-6.1249 -0.867 0>, rotation=<rot -90 0.0396 -0.0397>}]
 head [
  {shapeType=Capsule, length=0, radius=7.5, shapeName=head_capsule, center=<4.2324 0 0>}
  {shapeType=Capsule, length=3.5, radius=4, shapeName=head_capsule, center=<-1.6939 5.1865 0>}
 ]
 lowerarm_l [{
   shapeType=Capsule
   length=19.5
   radius=4
   shapeName=lowerarm_r_capsule
   center=<12.2258 -0.3619 0.8801>
   rotation=<rot -85.1201 14.9085 165.4289>
  }]
 lowerarm_r [{
   shapeType=Capsule
   length=19.5
   radius=4
   shapeName=lowerarm_r_capsule
   center=<-12.2258 0.3619 -0.8801>
   rotation=<rot -85.1201 14.9085 -14.5711>
  }]
 neck_01 [{shapeType=Capsule, length=7.5, radius=4.5, shapeName=neck_01_capsule, center=<1.5582 0 0>, rotation=<rot 90 -4.0447 -4.0445>}]
 spine_04 [{shapeType=Capsule, length=11, radius=7.5, shapeName=spine_04_capsule, center=<2.4912 3.6238 0>}]
 spine_05 [{shapeType=Capsule, length=8.5, radius=12, shapeName=spine_05_capsule, center=<0 -3.3752 0>, rotation=<rot 0 -0.0002 0>}]
 upperarm_l [{shapeType=Capsule, length=16.5, radius=4.5, shapeName=upperarm_l_capsule, center=<13.9087 -1.0491 0>, rotation=<rot -90 166.7373 -166.7374>}]
 upperarm_r [{shapeType=Capsule, length=18, radius=3, shapeName=upperarm_l_capsule, center=<-13.9087 1.0491 0>, rotation=<rot -60.0002 -89.9999 -90>}]
}
constraints [
 {bone1Name=dyn_earrings_l, bone2Name=head, swing1LimitDegrees=10, swing2LimitDegrees=10, twistLimitDegrees=5}
 {bone1Name=dyn_earrings_r, bone2Name=head, swing1LimitDegrees=10, swing2LimitDegrees=10, twistLimitDegrees=5}
 {bone1Name=dyn_curl, bone2Name=head, swing1LimitDegrees=30, swing2LimitDegrees=25, twistLimitDegrees=5}
 {bone1Name=dyn_bangs, bone2Name=head, swing1LimitDegrees=10, swing2LimitDegrees=5, twistLimitDegrees=2.5}
 {bone1Name=dyn_head_tail_2_r, bone2Name=dyn_head_tail_1_r, swing1LimitDegrees=45, swing2LimitDegrees=10, twistLimitDegrees=5}
 {bone1Name=dyn_head_tail_2_l, bone2Name=dyn_head_tail_1_l, swing1LimitDegrees=15, swing2LimitDegrees=5, twistLimitDegrees=5}
 {bone1Name=dyn_head_tail_3_r, bone2Name=dyn_head_tail_2_r, swing1LimitDegrees=35, swing2LimitDegrees=10, twistLimitDegrees=15}
 {bone1Name=dyn_head_tail_3_l, bone2Name=dyn_head_tail_2_l, swing1LimitDegrees=25, swing2LimitDegrees=20, twistLimitDegrees=25}
 {bone1Name=dyn_head_tail_4_r, bone2Name=dyn_head_tail_3_r, swing1LimitDegrees=30, swing2LimitDegrees=30, twistLimitDegrees=5}
 {bone1Name=dyn_head_tail_4_l, bone2Name=dyn_head_tail_3_l, swing1LimitDegrees=30, swing2LimitDegrees=15, twistLimitDegrees=25}
 {bone1Name=dyn_head_tail_1_r, bone2Name=dyn_head_tail_base_r, swing1LimitDegrees=60, swing2LimitDegrees=5, twistLimitDegrees=5}
 {bone1Name=dyn_head_tail_1_l, bone2Name=dyn_head_tail_base_l, swing1LimitDegrees=30, swing2LimitDegrees=5, twistLimitDegrees=5}
 {
  bone1Name=dyn_head_tail_base_l
  bone2Name=spine_05
  swing1LimitDegrees=45
  swing1Motion=Free
  swing2LimitDegrees=45
  swing2Motion=Free
  twistLimitDegrees=45
  twistMotion=Free
 }
 {
  bone1Name=dyn_head_tail_base_r
  bone2Name=spine_05
  swing1LimitDegrees=45
  swing1Motion=Free
  swing2LimitDegrees=45
  swing2Motion=Free
  twistLimitDegrees=45
  twistMotion=Free
 }
]
