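# Post-copy animation blueprint for the FierceBraid female character's face accessory, adding secondary motion to its dynamic bones.
# It layers the parent pose with a body mesh part cached on customization load, filtered by upper-arm and dragon bones, then runs a Control Rig.
# An open parachute enables additive offsets on both head-tail base bones.
# The cached result feeds two neck-space RigidBody simulations, standard and FrontEnd physics assets, selected by a bool blend.
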AnimBlueprint /BRCosmetics/Characters/Player/Female/Medium/Bodies/F_MED_FierceBraid/Meshes/Parts/F_MED_FierceBraid_FaceAcc_AnimBP
target_skeleton=/Game/Characters/Player/Female/Medium/Bodies/F_MED_FierceBraid/Meshes/Parts/F_MED_FierceBraid_FaceAcc_Skeleton
graphs [
 AnimGraph {nodes [
   n1 = RigidBody(ComponentPose=Local To Component(LocalPose=Use cached pose 'pre_dyn'())) {
    bClampLinearTranslationLimitToRefPose=true
    bOverrideWorldGravity=true
    overridePhysicsAsset=/Game/Characters/Player/Female/Medium/Bodies/F_MED_FierceBraid/Meshes/Parts/F_MED_FierceBraid_FaceAcc_Physics
    simulationSpace=BaseBoneSpace
    alphaBoolBlend {blendOption=Linear}
    alphaScaleBiasClamp {bInterpResult=true, interpSpeedDecreasing=100, interpSpeedIncreasing=100}
    baseBoneRef {boneName=neck_01}
    componentAppliedLinearAccClamp=<100 100 100>
    componentLinearAccScale=<1 1 1>
    componentLinearVelScale=<0.75 0.75 0.75>
    overrideWorldGravity {z=-980}
    simSpaceSettings {
     maxAngularAcceleration=300
     maxAngularVelocity=1
     maxLinearAcceleration=80
     maxLinearVelocity=250
     velocityScaleZ=0
     worldAlpha=0.25
     externalAngularVelocity=<1 1 1>
     externalLinearDragV=<0.1 0.1>
    }
   }
   n2 = RigidBody(ComponentPose=Local To Component(LocalPose=Use cached pose 'pre_dyn'())) {
    bClampLinearTranslationLimitToRefPose=true
    bOverrideWorldGravity=true
    bTransferBoneVelocities=true
    overridePhysicsAsset=/Game/Characters/Player/Female/Medium/Bodies/F_MED_FierceBraid/Meshes/Parts/F_MED_FierceBraid_FaceAcc_Physics_FrontEnd
    simulationSpace=BaseBoneSpace
    alphaBoolBlend {blendOption=Linear}
    alphaScaleBiasClamp {bInterpResult=true, interpSpeedDecreasing=100, interpSpeedIncreasing=100}
    baseBoneRef {boneName=neck_01}
    componentAppliedLinearAccClamp=<100 100 100>
    componentLinearAccScale=<1 1 1>
    componentLinearVelScale=<0.75 0.75 0.75>
    overrideWorldGravity {z=-980}
    simSpaceSettings {
     maxAngularAcceleration=300
     maxAngularVelocity=1
     maxLinearAcceleration=80
     maxLinearVelocity=250
     velocityScaleZ=0
     worldAlpha=0.25
     externalAngularVelocity=<1 1 1>
     externalLinearDragV=<0.1 0.1>
    }
   }
   Output Pose(Result=Blend Poses by bool(BlendPose_0=Component To Local(ComponentPose=@n2), BlendPose_1=Component To Local(ComponentPose=@n1)) {blendType=Linear}) {layerGroup=None}
   n3 = Layered blend per bone(
    BasePose=Copy Pose From Mesh() {bUseAttachedParent=true}
    BlendPoses_0=Copy Pose From Mesh(SourceMeshComponent=Get body() -> (body:Skeletal Mesh Component Budgeted Object Reference)) {bCopyCurves=true, bUseAttachedParent=true}
   ) {
    bMeshSpaceRotationBlend=true
    layerSetup [{branchFilters [
       {boneName=upperarm_r}
       {boneName=dyn_dragon_head}
       {boneName=upperarm_l}
       {boneName=dyn_dragon_spine_1}
       {boneName=dyn_dragon_spine_2}
       {boneName=dyn_dragon_spine_3}
       {boneName=dyn_dragon_spine_4}
       {boneName=dyn_dragon_spine_5}
      ]}]
   }
   n4 = Control Rig(Source=@n3) {
    alphaBoolBlend {blendOption=Linear}
    controlRigAssetReference {blueprintRigClass=/Game/Characters/Player/Female/Medium/Bodies/F_MED_FierceBraid/Meshes/Parts/F_MED_FierceBraid_FaceAcc_CtrlRig.F_MED_FierceBraid_FaceAcc_CtrlRig_C}
   }
   n5 = Get Dyn_tranform_alfa() -> (Dyn_tranform_alfa:double)
   n6 = "Transform (Modify) Bone - Bone: dyn_head_tail_base_r"(ComponentPose=Local To Component(LocalPose=@n4), Alpha=@n5, Translation="0,-6,-1", Rotation="-20,45,0", Scale="1,1,1") {
    rotationMode=BMM_Additive
    rotationSpace=BCS_BoneSpace
    translationMode=BMM_Additive
    translationSpace=BCS_BoneSpace
    alphaBoolBlend {blendOption=Linear}
    alphaScaleBiasClamp {bInterpResult=true}
    boneToModify {boneName=dyn_head_tail_base_r}
    rotation {pitch=-20, yaw=45}
    translation {y=-6, z=-1}
   }
   n7 = "Transform (Modify) Bone - Bone: dyn_head_tail_base_l"(ComponentPose=@n6, Alpha=@n5, Translation="2,5,-3", Rotation="20,-45,-20", Scale="1,1,1") {
    rotationMode=BMM_Additive
    rotationSpace=BCS_BoneSpace
    translationMode=BMM_Additive
    translationSpace=BCS_BoneSpace
    alphaBoolBlend {blendOption=Linear}
    alphaScaleBiasClamp {bInterpResult=true}
    boneToModify {boneName=dyn_head_tail_base_l}
    rotation=<rot 20 -45 -20>
    translation=<2 5 -3>
   }
   Save cached pose 'pre_dyn'(Pose=Component To Local(ComponentPose=@n7)) {cacheName=pre_dyn}
  ]}
 EventGraph {nodes [
   Event BlueprintUpdateAnimation()
   > n1 = Branch(Condition=Get bIsParachuteOpened())
   > Set Dyn_tranform_alfa(Dyn_tranform_alfa:double=1) -> (Output_Get:double)
   Set Dyn_tranform_alfa(execute=@n1.else, Dyn_tranform_alfa=0)
   n2 = Event OnCharacterCustomizationLoaded()
   > Set body(body:Skeletal Mesh Component Budgeted Object Reference="Get (a copy)"(Array:Array of Skeletal Mesh Component Budgeted Object References=@n2.CustomMeshParts, Dimension 1=1) -> (Output:Skeletal Mesh Component Budgeted Object Reference)) -> (Output_Get:Skeletal Mesh Component Budgeted Object Reference)
  ]}
]
variables [
 body {type=SkeletalMeshComponentBudgeted}
 Dyn_tranform_alfa {type=number}
]
